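# Unlit translucent sprite material for Festivus lens-flare particles that pulse with the event's music.
# Particle color drives emissive, while two powered radial sphere masks plus a spiky flare texture form the flare shape.
# That shape is scaled by particle alpha, a near-camera distance fade and depth fade, which mobile skips.
# Beat and breakout values from two parameter collections modulate it, clamped so it never fully drops out.
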Material /BarnTowelGameplay/Athena/Effects/Materials/Festivus/M_Festivus_Flare_Sprite
graph {
 properties {
  bUsedWithNiagaraSprites=true
  bUsedWithParticleSprites=true
  blendMode=BLEND_Translucent
  pixelDepthOffsetMode=PDOM_Legacy
  shadingModel=MSM_Unlit
  usedShadingModels=Unlit
 }
 nodes [
  MaterialFunctionCall(Emissive=ParticleColor().RGB) {materialFunction=&MF_FX_EmissiveAdjustments}
  n1 = CollectionParameter() {collection=/Game/Athena/Environments/Festivus/Materials/FFT_Materials/FFT_Breakout_MCP, parameterName=FFT_01_Smothed}
  n2 = Add(
   A=CollectionParameter() {collection=/Game/Athena/Environments/Festivus/MotorizedLights/Lighting_MPC, parameterName=1_Beat_1}
   B=Multiply(
    A=MaterialFunctionCall(Alpha=@n1) {materialFunction=&SmoothStep}.Result
    B=ScalarParameter() {defaultValue=1, parameterName=Beat_Intensity}
   )
  )
  Multiply(A=@n2) {constB=150}
  n3 = Constant() {r=0.5}
  n4 = SphereMask(A=TextureCoordinate(), B=@n3, Radius=@n3) {hardnessPercent=0}
  n5 = Multiply(A=Power(Base=@n4) {constExponent=25}) {constB=2}
  n6 = LinearInterpolate(B=Constant() {r=1}) {constA=1, constAlpha=0.7}
  n7 = Multiply(A=Power(Base=@n4) {constExponent=4}) {constB=0.15}
  n8 = TextureSampleParameter2D() {parameterName=FlareMask, texture=/Game/Effects/Fort_Effects/Textures/Flares/T_Flare_Spikeball}
  n9 = Power(Base=MaterialFunctionCall(Distance=ScalarParameter() {defaultValue=25000, parameterName=NearFade}) {materialFunction=&Near_Camera_Fade}.Result) {constExponent=8}
  n10 = Multiply(A=Multiply(A=Constant() {r=1}, B=ParticleColor().A), B=LinearInterpolate(A=Constant() {r=1}, B=Constant() {r=0.25}, Alpha=Clamp(None=@n9)))
  n11 = DepthFade(Opacity=@n10, FadeDistance=ScalarParameter() {defaultValue=500, parameterName=FadeDistance})
  Multiply(
   A=Add(A=Multiply(A=@n5, B=@n6), B=Add(A=@n7, B=@n8.RGB))
   B=Multiply(A=ShadingPathSwitch(Default=@n11, Mobile=@n10) {desc=Migrated from FeatureLevelSwitch}, B=Clamp(None=@n2) {minDefault=0.5})
  )
 ]
}
